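# Skydiving contrail cosmetic that surrounds the diver with glowing blue dragonfly meshes, limb trails, rings, glows, decal sprites and GPU speedlines.
# Emitters sample User.SkeletalMesh bones or sockets by partition index to track hands and feet.
# The system Timeline scales brightness, trail width, speedlines and dragonfly motion.
# Front-end modifiers retune values, bIsFrontEndPreview swaps in separate dragonflies without inherited movement, and Variant_Float sets trail visibility tags.
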NiagaraSystem /BRCosmetics/Athena/Effects/Particles/Skydiving/NS_Contrail_ZirconSweep/NS_Contrail_ZirconSweep
dependencies {
 usedDataInterfaces [
  &NiagaraDataInterfaceCurve
  &NiagaraDataInterfaceVector2DCurve
  &NiagaraDataInterfaceArrayFloat3
  &NiagaraDataInterfaceArrayFloat
 ]
 usedDynamicInputs [&SetBoolByFloatComparison, &TransformVector]
 usedModules [
  /Game/Effects/Niagara/Modules/Skydiving/FN_DivingAttributes/v4/FN_DivingAttributes_v4
  &SystemState
  &self:SystemSpawnScript.SystemScriptSource.SystemScriptGraph.NiagaraNodeAssignment_0.SetVariables_#1
  &self:SystemSpawnScript.SystemScriptSource.SystemScriptGraph.NiagaraNodeAssignment_1.SetVariables_#2
  &self:SystemSpawnScript.SystemScriptSource.SystemScriptGraph.NiagaraNodeAssignment_2.SetVariables_#3
  &self:SystemSpawnScript.SystemScriptSource.SystemScriptGraph.NiagaraNodeAssignment_3.SetVariables_#4
  &self:SystemSpawnScript.SystemScriptSource.SystemScriptGraph.NiagaraNodeAssignment_4.SetVariables_#5
  &Timeline
  &FN_Integrator
 ]
}
emitters {
 Dragonflies {
  emitterUpdateScript [
   &EmitterState(
    Inactive Response=::NewEnumerator1
    Life Cycle Mode=::NewEnumerator1
    Loop Duration=&RandomRangeFloat(Minimum=0.75, Maximum=1.5)
    UseLoopDelay=false
   )
   &SpawnRate(
    SpawnRate=&MakeCustomFloatFromBool(
     Bool=$User.bIsFrontEndPreview
     True Float=0
     False Float=&MakeFloatFromVector(
      Vector=&VectorFromFloat(Value=&Lerp_Float(A=6, B=10, Alpha=$System.Timeline.NormalizedTime))
      Channel=::NewEnumerator0
     )
    )
    Use Spawn Probability=false
   )
  ]
  particleSpawnScript [
   &PartitionParticles(Number Of Partitions=4)
   &InitializeParticle(
    Color Maximum=<rgba 0.0348 0.5965 0.96 1>
    Color Minimum=<rgba 0.0232 0.22 0.4135 1>
    Color Mode=::NewEnumerator2
    Lifetime Max=1.5
    Lifetime Mode=::NewEnumerator1
    Mesh Scale Mode=::NewEnumerator1
    Mesh Uniform Scale Max=1
    Mesh Uniform Scale Min=0.5
    UsePositionOffset=false
    Write Material Random=true
   )
   &SkeletalMeshLocation(
    Bone / Socket Index=$Particles.PartitionParticles.Partition
    MaskSpawnGroup=false
    Mesh Sampling Type=::NewEnumerator5
    Orientation Sampling=::NewEnumerator1
    Skeletal Mesh=$User.SkeletalMesh
    Socket Sampling Mode=::NewEnumerator5
   )
   &CylinderLocation(
    Cylinder Height=15
    Cylinder Height Midpoint=1
    Cylinder Radius=&Lerp_Float(A=20, B=40, Alpha=$System.Timeline.NormalizedTime)
    MaskSpawnGroup=false
    Offset=<0 0 30>
    Offset Mode:ENiagara_OffsetMode=::NewEnumerator0
    Override Seed=false
    Surface Only=true
    Use Endcaps In Surface Only Mode=false
    UseOwnerScale=true
   )
   &InitialMeshOrientation(
    Mesh Axis To Orient=<0 1 0>
    Mesh Orientation Mode=::NewEnumerator2
    Rotation=&RandomRangeVector(Minimum=<0 -0.2 0>, Maximum=<0 0.2 0>)
    Use Rotation Vector=true
   )
  ]
  particleUpdateScript [
   &InheritSourceMovement(
    Applied Position Scale=&Lerp_Vector(
     A=<0.98 0.98 0.99>
     B=&RandomRangeVector(Minimum=<0.98 0.98 0.98>, Maximum=<0.985 0.985 0.985>)
     Alpha=$System.Timeline.NormalizedTime
    )
    ApplyForce=false
    ApplyPosition=true
    ApplyPositionFalloff=false
   )
   &ParticleState()
   &InheritVelocity(
    Inherited Velocity Amount Scale=&Lerp_Vector(A=<0.5 0.5 0>, B=<1 1 1>, Alpha=$System.Timeline.NormalizedTime)
    Inherited Velocity Speed Limit=1
    Speed Limit=true
    UseSpeedThreshold=false
   )
   &WindForce(
    Offset By Time:Vector3f=<0 0 -10>
    Scale=0.6089
    Wind Speed=&RandomRangeVector(Minimum=<-100 -100 50>, Maximum=<100 100 50>)
    Wind Speed Scale=&Lerp_Float(A=1, B=0, Alpha=$System.Timeline.NormalizedTime)
   )
   &VortexForce(
    UseOriginOffset=false
    Vortex Force Amount=&Lerp_Float(A=200, B=0, Alpha=$System.Timeline.NormalizedTime)
    Vortex Origin:NiagaraPosition=$Particles.SkeletalMeshLocation.SampledPosition
   )
   &Drag(Drag=&RandomRangeFloat(Minimum=1, Maximum=2), Use Linear Drag=true, Use Rotational Drag=true)
   &SolveForcesAndVelocity(Clamp Velocity=false, Limit Acceleration=false)
   &ScaleMeshSize(Scale Factor=&VectorFromFloat(Value=&FloatFromCurve(
      FloatCurve=DataInterface() {propertyValues {Curve {keys [
          {arriveTangent=7.9421, interpMode=RCIM_Cubic, leaveTangent=7.9421, tangentMode=RCTM_User, time=0, value=0}
          {interpMode=RCIM_Cubic, tangentMode=RCTM_User, time=0.5, value=1}
          {arriveTangent=-6.2155, interpMode=RCIM_Cubic, leaveTangent=-6.2155, tangentMode=RCTM_User, time=1, value=0}
         ]}}}
      Scale Curve=2
     )))
   &ScaleColor(
    Scale Alpha=0.9
    Scale RGB=&VectorFromFloat(Value=&Lerp_Float(A=1, B=3, Alpha=$System.Timeline.NormalizedTime))
    ScaleA=true
    ScaleRGB=true
   )
   &DynamicMaterialParameters(
    Index 0 Param 1=1.5
    Index 0 Param 2=&Lerp_Float(A=4, B=7, Alpha=$System.Timeline.NormalizedTime)
    Index 0 Param 3=0.5
    Index 0 Param 4=-0.2
    Module.Param0WriteEnabled=true
    Module.Param1WriteEnabled=true
    Module.Param2WriteEnabled=true
    Module.Param3WriteEnabled=true
   ) {module=DynamicMaterialParameters001}
  ]
  rendererClasses [&NiagaraMeshRendererProperties]
  renderers [{rendererClass=&NiagaraMeshRendererProperties, rendererIndex=0}]
 }
 Dragonflies_FE {
  emitterUpdateScript [
   &EmitterState(
    Inactive Response=::NewEnumerator1
    Life Cycle Mode=::NewEnumerator1
    Loop Duration=&RandomRangeFloat(Minimum=0.75, Maximum=1.5)
    UseLoopDelay=false
   )
   &SpawnRate(
    SpawnRate=&MakeCustomFloatFromBool(
     Bool=$User.bIsFrontEndPreview
     True Float=&MakeFloatFromVector(
      Vector=&VectorFromFloat(Value=&Lerp_Float(A=6, B=10, Alpha=$System.Timeline.NormalizedTime))
      Channel=::NewEnumerator0
     )
     False Float=0
    )
    Use Spawn Probability=false
   )
  ]
  particleSpawnScript [
   &PartitionParticles(Number Of Partitions=4)
   &InitializeParticle(
    Color Maximum=<rgba 0.0348 0.5965 0.96 1>
    Color Minimum=<rgba 0.0232 0.22 0.4135 1>
    Color Mode=::NewEnumerator2
    Lifetime Max=1.5
    Lifetime Mode=::NewEnumerator1
    Mesh Scale Mode=::NewEnumerator1
    Mesh Uniform Scale Max=1
    Mesh Uniform Scale Min=0.5
    UsePositionOffset=false
    Write Material Random=true
   )
   &SkeletalMeshLocation(
    Bone / Socket Index=$Particles.PartitionParticles.Partition
    MaskSpawnGroup=false
    Mesh Sampling Type=::NewEnumerator5
    Orientation Sampling=::NewEnumerator1
    Skeletal Mesh=$User.SkeletalMesh
    Socket Sampling Mode=::NewEnumerator5
   )
   &CylinderLocation(
    Cylinder Height=15
    Cylinder Height Midpoint=1
    Cylinder Radius=&Lerp_Float(A=20, B=40, Alpha=$System.Timeline.NormalizedTime)
    MaskSpawnGroup=false
    Offset=<0 0 30>
    Offset Mode:ENiagara_OffsetMode=::NewEnumerator0
    Override Seed=false
    Surface Only=true
    Use Endcaps In Surface Only Mode=false
    UseOwnerScale=true
   )
   &InitialMeshOrientation(
    Mesh Axis To Orient=<0 1 0>
    Mesh Orientation Mode=::NewEnumerator2
    Rotation=&RandomRangeVector(Minimum=<0 -0.2 0>, Maximum=<0 0.2 0>)
    Use Rotation Vector=true
   )
  ]
  particleUpdateScript [
   &ParticleState()
   &UpdateMeshOrientation(
    Facing Direction=&Lerp_Vector(A=<1 0 0>, B=<-1 0 0>, Alpha=$System.Timeline.NormalizedTime)
    Facing Interpolation=false
    Orientation Method=::NewEnumerator1
    Override Quat=false
    TransformFacing=false
    TransformUp=false
   )
   &WindForce(
    Offset By Time:Vector3f=<0 0 -10>
    Scale=0.6089
    Wind Speed=&RandomRangeVector(Minimum=<-100 -100 50>, Maximum=<100 100 50>)
    Wind Speed Scale=&Lerp_Float(A=1, B=0, Alpha=$System.Timeline.NormalizedTime)
   )
   &VortexForce(
    UseOriginOffset=false
    Vortex Force Amount=&Lerp_Float(A=200, B=0, Alpha=$System.Timeline.NormalizedTime)
    Vortex Origin:NiagaraPosition=$Particles.SkeletalMeshLocation.SampledPosition
   )
   &Drag(Drag=&RandomRangeFloat(Minimum=1, Maximum=2), Use Linear Drag=true, Use Rotational Drag=true)
   /SetVariables(Particles.Velocity=&Lerp_Vector(
     A=&Add_Vector(A=$Particles.Velocity, B=<0 0 1.5>)
     B=&Add_Vector(A=$Particles.Velocity, B=<0 -4 5>)
     Alpha=$System.Timeline.NormalizedTime
    )) {isSetParameters=true}
   &SolveForcesAndVelocity(Clamp Velocity=false, Limit Acceleration=false)
   &ScaleMeshSize(Scale Factor=&VectorFromFloat(Value=&FloatFromCurve(
      FloatCurve=DataInterface() {propertyValues {Curve {keys [
          {arriveTangent=7.9421, interpMode=RCIM_Cubic, leaveTangent=7.9421, tangentMode=RCTM_User, time=0, value=0}
          {interpMode=RCIM_Cubic, tangentMode=RCTM_User, time=0.5, value=1}
          {arriveTangent=-6.2155, interpMode=RCIM_Cubic, leaveTangent=-6.2155, tangentMode=RCTM_User, time=1, value=0}
         ]}}}
      Scale Curve=2
     )))
   &ScaleColor(
    Scale Alpha=0.9
    Scale RGB=&VectorFromFloat(Value=&Lerp_Float(A=1, B=3, Alpha=$System.Timeline.NormalizedTime))
    ScaleA=true
    ScaleRGB=true
   )
   &DynamicMaterialParameters(
    Index 0 Param 1=1.5
    Index 0 Param 2=&Lerp_Float(A=4, B=7, Alpha=$System.Timeline.NormalizedTime)
    Index 0 Param 3=0.5
    Index 0 Param 4=-0.2
    Module.Param0WriteEnabled=true
    Module.Param1WriteEnabled=true
    Module.Param2WriteEnabled=true
    Module.Param3WriteEnabled=true
   ) {module=DynamicMaterialParameters001}
  ]
  rendererClasses [&NiagaraMeshRendererProperties]
  renderers [{rendererClass=&NiagaraMeshRendererProperties, rendererIndex=0}]
 }
 GlowFeet {
  emitterUpdateScript [
   &EmitterState(Life Cycle Mode=::NewEnumerator1, Loop Behavior=::NewEnumerator1, Loop Duration Mode=::NewEnumerator1, UseLoopDelay=false)
   &SpawnBurst_Instantaneous(Spawn Count=2, Use Spawn Probability=false)
  ]
  particleSpawnScript [
   &FN_SetSpawnGroup(Number of Spawn Groups=2) {enabled=false}
   &PartitionParticles()
   &InitializeParticle(
    Color=<rgba 0.0549 0.5333 1 0.1>
    Lifetime Max=0.5
    Lifetime Min=0.3
    Lifetime Mode=::NewEnumerator1
    Sprite Size=<50 30>
    Sprite Size Mode=::NewEnumerator4
    UsePositionOffset=false
    Write Material Random=true
   )
   &CameraOffset(Camera Offset Amount=5)
  ]
  particleUpdateScript [
   &ParticleState(Kill Particles When Lifetime Has Elapsed=false, Let Infinitely Lived Particles Die When Emitter Deactivates=true)
   &SkeletalMeshLocation(
    Bone / Socket Index=&Add_Integer(A=$Particles.PartitionParticles.Partition, B=2)
    MaskSpawnGroup=false
    Mesh Sampling Type=::NewEnumerator5
    Orientation Sampling=::NewEnumerator1
    Sampled Position Offset=&FN_FrontEndVector3Modifier(Gameplay=<-3 0 0>, Front End=<0 0 0>)
    Skeletal Mesh=$User.SkeletalMesh
    Socket Sampling Mode=::NewEnumerator5
   )
   &ScaleSpriteSize(Scale Factor=&Vector2DFromCurve(
     Vector2Curve=DataInterface() {propertyValues {XCurve {keys [<at 0 0.6>,<at 1 1>]}, YCurve {keys [<at 0 0.03>,<at 1 1.75>]}}}
     ScaleCurve=<1 1>
    ))
   &ScaleColor(
    Scale Alpha=&FN_FrontEndFloatModifier(Gameplay=1, Front End=1)
    Scale RGB=&VectorFromFloat(Value=&Lerp_Float(A=1, B=2, Alpha=$System.Timeline.NormalizedTime))
    ScaleA=true
    ScaleRGB=true
   )
  ]
  rendererClasses [&NiagaraSpriteRendererProperties]
  renderers [{rendererClass=&NiagaraSpriteRendererProperties}]
 }
 GlowHands {
  emitterUpdateScript [
   &EmitterState(Life Cycle Mode=::NewEnumerator1, Loop Behavior=::NewEnumerator1, Loop Duration Mode=::NewEnumerator1, UseLoopDelay=false)
   &SpawnBurst_Instantaneous(Spawn Count=2, Use Spawn Probability=false)
  ]
  particleSpawnScript [
   &FN_SetSpawnGroup(Number of Spawn Groups=2) {enabled=false}
   &PartitionParticles()
   &InitializeParticle(
    Color=<rgba 0.056 0.532 1 0.1>
    Lifetime Max=0.5
    Lifetime Min=0.3
    Lifetime Mode=::NewEnumerator1
    Sprite Size=<40 30>
    Sprite Size Mode=::NewEnumerator4
    UsePositionOffset=false
    Write Material Random=true
   )
   &CameraOffset(Camera Offset Amount=5)
  ]
  particleUpdateScript [
   &ParticleState(Kill Particles When Lifetime Has Elapsed=false, Let Infinitely Lived Particles Die When Emitter Deactivates=true)
   &SkeletalMeshLocation(
    Bone / Socket Index=$Particles.PartitionParticles.Partition
    MaskSpawnGroup=false
    Mesh Sampling Type=::NewEnumerator5
    Orientation Sampling=::NewEnumerator1
    Sampled Position Offset=&FN_FrontEndVector3Modifier(Gameplay=<0 5 9>, Front End=<0 0 0>)
    Sampled Position Offset Coordinate Space:ENiagaraSkelMeshTransforms=::NewEnumerator1
    Skeletal Mesh=$User.SkeletalMesh
    Socket Sampling Mode=::NewEnumerator5
   )
   &ScaleSpriteSize(Scale Factor=&Vector2DFromCurve(
     Vector2Curve=DataInterface() {propertyValues {XCurve {keys [<at 0 0.6>,<at 1 1>]}, YCurve {keys [<at 0 0.03>,<at 1 1.75>]}}}
     ScaleCurve=<1 1>
    ))
   &ScaleColor(
    Scale Alpha=&FN_FrontEndFloatModifier(Gameplay=1, Front End=1)
    Scale RGB=&VectorFromFloat(Value=&Lerp_Float(A=1, B=2, Alpha=$System.Timeline.NormalizedTime))
    ScaleA=true
    ScaleRGB=true
   )
  ]
  rendererClasses [&NiagaraSpriteRendererProperties]
  renderers [{rendererClass=&NiagaraSpriteRendererProperties}]
 }
 ParDecal {
  emitterUpdateScript [
   &EmitterState(Life Cycle Mode=::NewEnumerator1, Loop Behavior=::NewEnumerator1, Loop Duration Mode=::NewEnumerator1, UseLoopDelay=false)
   &SpawnBurst_Instantaneous(Spawn Count=4, Use Spawn Probability=false)
  ]
  particleSpawnScript [
   &InitializeParticle(Color=<rgba 0.112 1.064 2 1>, Sprite Size Mode=::NewEnumerator3, Uniform Sprite Size=25, UsePositionOffset=false, Write Material Random=true)
   &CameraOffset(Camera Offset Amount=20)
   &PartitionParticles()
  ]
  particleUpdateScript [
   &ParticleState(Kill Particles When Lifetime Has Elapsed=false, Let Infinitely Lived Particles Die When Emitter Deactivates=true)
   &SkeletalMeshLocation(
    Apply Sampled Orientation to Mesh Particles=true
    Bone / Socket Index=&Add_Integer(A=$Particles.PartitionParticles.Partition, B=2)
    Bone Sampling Mode=::NewEnumerator4
    Inherit Velocity=false
    MaskSpawnGroup=false
    Sampled Position Offset=&SelectVectorFromArray(
     Direct Array Index=$Particles.PartitionParticles.Partition
     Vector Selection Array=DataInterface() {propertyValues {FloatData [<0 10 0>, <0 -10 0>, <0 10 0>, <0 -10 0>]}}
    )
    Sampled Position Offset Coordinate Space:ENiagaraMeshTransforms=::NewEnumerator3
    Skeletal Mesh=$User.SkeletalMesh
   )
   &ScaleColor(
    Scale Alpha=&Lerp_Float(A=1, B=1, Alpha=$System.Timeline.NormalizedTime)
    Scale RGB=&VectorFromFloat(Value=&Lerp_Float(A=1, B=2, Alpha=$System.Timeline.NormalizedTime))
    ScaleA=true
    ScaleRGB=true
   )
   &ScaleColor(
    Scale Alpha=&Multiply_Float(
     A=&FNTimeOfDayScalar(Night Scale=0.5, Morning Scale=0.7, Evening Scale=0.7)
     B=&FN_FrontEndFloatModifier(Gameplay=1, Front End=0.2)
    )
    Scale RGB=&FN_FrontEndVector3Modifier(Gameplay=<1 1 1>, Front End=<2 2 2>)
    ScaleA=true
    ScaleRGB=true
   ) {module=ScaleColor001}
  ]
  rendererClasses [&NiagaraSpriteRendererProperties]
  renderers [{rendererClass=&NiagaraSpriteRendererProperties}]
 }
 Rings {
  emitterUpdateScript [
   &EmitterState(
    Inactive Response=::NewEnumerator1
    Life Cycle Mode=::NewEnumerator1
    Loop Behavior=::NewEnumerator1
    Loop Duration Mode=::NewEnumerator1
    UseLoopDelay=false
   )
   &SpawnBurst_Instantaneous(Spawn Count=12, Use Spawn Probability=false, UseLoopCountLimit=false)
  ]
  particleSpawnScript [
   &InitializeParticle(
    Color=<rgba 0 2 1.1379 1>
    Mesh Scale=&Multiply_VectorByFloat(Vector=<1.8 1.8 3>, Float=0.1)
    Mesh Scale Mode=::NewEnumerator4
    UsePositionOffset=false
    Write Material Random=true
   )
   &PartitionParticles(Number Of Partitions=4)
   &SubUVAnimation(End Frame=1, SubUV Animation Mode:ENiagara_SubUVLookupMode=::NewEnumerator2)
   &InitialMeshOrientation(
    Mesh Orientation Mode=::NewEnumerator2
    Orientation Vector=<0 0 1>
    Rotation=&RandomRangeVector(Minimum=<0 0 0>, Maximum=<0 0 1>)
    Use Rotation Vector=true
   )
   &CameraOffset()
  ]
  particleUpdateScript [
   &ParticleState(Kill Particles When Lifetime Has Elapsed=false)
   &SkeletalMeshLocation(
    Apply Sampled Orientation to Mesh Particles=true
    Bone / Socket Index=&Add_Integer(A=$Particles.PartitionParticles.Partition, B=0)
    Inherit Velocity=false
    Interpolate Mesh Sampling=false
    MaskSpawnGroup=false
    Mesh Sampling Type=::NewEnumerator5
    Sampled Position Offset=&FN_FrontEndVector3Modifier(
     Gameplay=&Lerp_Vector(A=<-3 0 0>, B=<-3 0 0>, Alpha=&MakeFloatFromInt(Int=$Particles.PartitionParticles.Partition))
     Front End=&Lerp_Vector(A=<-3 0 0>, B=<-3 0 0>, Alpha=&MakeFloatFromInt(Int=$Particles.PartitionParticles.Partition))
    )
    Sampled Position Offset Coordinate Space:ENiagaraMeshTransforms=::NewEnumerator3
    Skeletal Mesh=$User.SkeletalMesh
    Socket Sampling Mode=::NewEnumerator5
   )
   &ScaleMeshSize(Scale Factor=&Lerp_Vector(A=<1 1 1>, B=<1 1 1>, Alpha=$System.Timeline.NormalizedTime))
   &DynamicMaterialParameters(
    Index 0 Param 1=0.0727
    Index 0 Param 4=4
    Index 1 Param 1=2.2152
    Index 1 Param 2=1
    Index 1 Param 3=1
    Index 1 Param 4=1
    Module.Param0WriteEnabled=true
    Module.Param0WriteEnabled1=true
    Module.Param1WriteEnabled=true
    Module.Param1WriteEnabled1=true
    Module.Param2WriteEnabled=true
    Module.Param2WriteEnabled1=true
    Module.Param3WriteEnabled=true
    Module.Param3WriteEnabled1=true
    Write Parameter Index 1=true
   )
   &ScaleColor(
    Scale Alpha=&FN_FrontEndFloatModifier(Gameplay=1, Front End=0.2)
    Scale RGB=&VectorFromFloat(Value=&Lerp_Float(A=2, B=3, Alpha=$System.Timeline.NormalizedTime))
    ScaleA=true
    ScaleRGB=true
   )
  ]
  rendererClasses [&NiagaraMeshRendererProperties]
  renderers [{rendererClass=&NiagaraMeshRendererProperties, rendererIndex=0}]
 }
 Speedlines {
  simTarget=GPUComputeSim
  emitterUpdateScript [
   &EmitterState(Life Cycle Mode=::NewEnumerator1, Loop Duration=1, UseLoopDelay=false)
   &SpawnRate(
    SpawnRate=&Multiply_Float(
     A=$System.Timeline.NormalizedTime
     B=&FN_FrontEndFloatModifier(Gameplay=&UniformRangedFloat(Minimum=100, Maximum=200), Front End=0)
    )
    Use Spawn Probability=false
   )
  ]
  particleSpawnScript [
   /Niagara/Modules/Spawn/Initialization/InitializeParticle(
    Color=&MakeLinearColorFromVectorAndFloat(
     "Vector (RGB)"=<1 1 1>
     "Float (Alpha)"=&Multiply_Float(
      A=&Multiply_Float(
       A=&UniformRangedFloat(Minimum=0.015, Maximum=0.15)
       B=&FN_FrontEndFloatModifier(Gameplay=1, Front End=0.5)
      )
      B=&Add_Float(
       A=1
       B=&Multiply_Float(A=10, B=&Subtract_Float(A=1, B=$Engine.Emitter.SpawnCountScale))
      )
     )
    )
    Lifetime=&FN_FrontEndFloatModifier(
     Gameplay=&UniformRangedFloat(Minimum=0.15, Maximum=0.3)
     Front End=&UniformRangedFloat(Minimum=0.1, Maximum=0.2)
    )
    Module.Write Color=true
    Module.Write Lifetime=true
    Module.Write Mass=true
    Module.Write Material Random=true
    Module.Write Position=false
    Module.Write Scale=false
    Module.Write SpriteRotation=false
    Module.Write SpriteSize=true
    Position=&SimulationPosition()
    Sprite Size=&UniformRangedVector2D(Minimum=<3 4>, Maximum=<5 6>)
   )
   &CylinderLocation(
    Cylinder Radius=&UniformRangedFloat(Minimum=100, Maximum=150)
    MaskSpawnGroup=false
    Override Seed=false
    UseOwnerScale=true
   )
   &AddVelocity(Velocity=&FN_FrontEndVector3Modifier(Gameplay=<0 0 1000>, Front End=<0 0 3000>))
   &AddVelocityFromPoint(
    Module.Use Falloff Distance=false
    Velocity Strength=&Multiply_Float(
     A=&FN_FrontEndFloatModifier(Gameplay=1, Front End=2)
     B=&UniformRangedFloat(Minimum=500, Maximum=700)
    )
   )
  ]
  particleUpdateScript [
   &ParticleState()
   &SolveForcesAndVelocity(Clamp Velocity=false, Limit Acceleration=false)
   &ScaleSpriteSize(Scale Factor=&Vector2DFromCurve(
     Vector2Curve=DataInterface() {propertyValues {XCurve {keys [<at 0 0.5 RCIM_Cubic>,<at 0.5 1 RCIM_Cubic>]}, YCurve {keys [<at 0 0 RCIM_Cubic>,<at 0.5 1 RCIM_Cubic>]}}}
     ScaleCurve=&Vector2DFromFloat(Value=1)
    ))
   &ScaleSpriteSizeBySpeed(Max Scale Factor=<1 40>, Min Scale Factor=<0.01 0.05>, Velocity Threshold=1)
   &ScaleColor(
    Scale Alpha=&FloatFromCurve(
     FloatCurve=DataInterface() {propertyValues {Curve {keys [<at 0 0 RCIM_Cubic>, <at 0.2 1>, {arriveTangent=-2.221, interpMode=RCIM_Cubic, leaveTangent=-2.221, tangentMode=RCTM_User, time=1, value=0}]}}}
     Scale Curve=&FN_FrontEndFloatModifier(Gameplay=&Lerp_Float(A=0, B=0.4, Alpha=$System.Timeline.NormalizedTime), Front End=0)
    )
    ScaleA=true
    ScaleRGB=true
   )
  ]
  rendererClasses [&NiagaraSpriteRendererProperties]
  renderers [{rendererClass=&NiagaraSpriteRendererProperties}]
 }
 Trails_Feet {
  emitterUpdateScript [
   &EmitterState(
    Inactive Response=::NewEnumerator1
    Life Cycle Mode=::NewEnumerator1
    Loop Behavior=::NewEnumerator1
    Loop Duration Mode=::NewEnumerator1
    UseLoopDelay=false
   )
   &SpawnBurst_Instantaneous(Spawn Count=8, Use Spawn Probability=false, UseLoopCountLimit=false)
  ]
  particleSpawnScript [
   &PartitionParticles()
   /Niagara/Modules/Spawn/Initialization/V2/InitializeParticle(
    Color Minimum=<rgba 0.501 0.501 0.501 1>
    Color Mode=::NewEnumerator2
    Mesh Scale=&RandomRangeVector(Minimum=<0.15 9 1>, Maximum=<0.2 11 1>)
    Mesh Scale Mode=::NewEnumerator4
    UsePositionOffset=false
    Write Material Random=true
   )
   /SetVariables(Particles.VisibilityTag=&MakeIntFromFloat(Float=$User.Variant_Float)) {isSetParameters=true, module=SetVariables_#6}
   &InitialMeshOrientation(
    Mesh Axis To Orient=<0 0 -1>
    Mesh Orientation Mode=::NewEnumerator2
    Rotation=&RandomRangeVector(Minimum=<0 -0.025 -0.04>, Maximum=<1 0.025 0.04>)
    Use Rotation Vector=true
   )
  ]
  particleUpdateScript [
   &ParticleState(Kill Particles When Lifetime Has Elapsed=false)
   &SkeletalMeshLocation(
    Bone / Socket Index=&Add_Integer(A=$Particles.PartitionParticles.Partition, B=2)
    Inherit Velocity=false
    MaskSpawnGroup=false
    Mesh Sampling Type=::NewEnumerator5
    Sampled Position Offset=&FN_FrontEndVector3Modifier(
     Gameplay=&Lerp_Vector(
      A=&RandomRangeVector(Minimum=<0 -1.5 -3>, Maximum=<0 1.3 0>)
      B=&RandomRangeVector(Minimum=<0 -1.5 -3>, Maximum=<0 1.5 0>)
      Alpha=&MakeFloatFromInt(Int=$Particles.PartitionParticles.Partition)
     )
     Front End=&Lerp_Vector(
      A=&RandomRangeVector(Minimum=<0 -1.5 -3>, Maximum=<0 1.3 0>)
      B=&RandomRangeVector(Minimum=<0 -1.5 -3>, Maximum=<0 1.5 0>)
      Alpha=&MakeFloatFromInt(Int=$Particles.PartitionParticles.Partition)
     )
    )
    Sampled Position Offset Coordinate Space:ENiagaraMeshTransforms=::NewEnumerator3
    Skeletal Mesh=$User.SkeletalMesh
    Socket Sampling Mode=::NewEnumerator5
   )
   &ScaleMeshSize(Scale Factor=&Lerp_Vector(A=<1 1 1>, B=<2.3 1 1>, Alpha=$System.Timeline.NormalizedTime))
   &UpdateMeshOrientation(
    Facing Interpolation=false
    Orientation Method=::NewEnumerator1
    Override Quat=false
    TransformFacing=true
    TransformUp=false
    Up Direction:Vector3f=&FN_FrontEndVector3Modifier(
     Gameplay=&Lerp_Vector(A=<0 -0.05 1>, B=<0 -0.2 1>, Alpha=$System.Timeline.NormalizedTime)
     Front End=&Lerp_Vector(A=<0 -0.05 1>, B=<0 -0.8 1>, Alpha=$System.Timeline.NormalizedTime)
    )
   )
   &FN_Timer(Increment Multiplier:NiagaraFloat=&Lerp_Float(A=1, B=3, Alpha=$System.Timeline.Time))
   &DynamicMaterialParameters(
    Index 1 Param 2=1
    Index 2 Param 3=1
    Index 2 Param 4=1
    Index 3 Param 1=&Add_Float(A=$Particles.FN_Timer.Value, B=&RandomRangeFloat(Minimum=0, Maximum=10))
    Module.Param0WriteEnabled=true
    Module.Param0WriteEnabled1=true
    Module.Param0WriteEnabled2=true
    Module.Param0WriteEnabled3=true
    Module.Param1WriteEnabled=true
    Module.Param1WriteEnabled1=true
    Module.Param1WriteEnabled2=true
    Module.Param1WriteEnabled3=true
    Module.Param2WriteEnabled=true
    Module.Param2WriteEnabled1=true
    Module.Param2WriteEnabled2=true
    Module.Param2WriteEnabled3=true
    Module.Param3WriteEnabled=true
    Module.Param3WriteEnabled1=true
    Module.Param3WriteEnabled2=true
    Module.Param3WriteEnabled3=true
    Write Parameter Index 1=true
    Write Parameter Index 2=true
    Write Parameter Index 3=true
   )
   &ScaleColor(
    Scale RGB=&VectorFromFloat(Value=&Lerp_Float(A=1, B=3, Alpha=$System.Timeline.NormalizedTime))
    ScaleA=true
    ScaleRGB=true
   )
  ]
  rendererClasses [&NiagaraMeshRendererProperties]
  renderers [{rendererClass=&NiagaraMeshRendererProperties, rendererIndex=0}]
 }
 Trails_Hand {
  emitterUpdateScript [
   &EmitterState(
    Inactive Response=::NewEnumerator1
    Life Cycle Mode=::NewEnumerator1
    Loop Behavior=::NewEnumerator1
    Loop Duration Mode=::NewEnumerator1
    UseLoopDelay=false
   )
   &SpawnBurst_Instantaneous(Spawn Count=8, Use Spawn Probability=false, UseLoopCountLimit=false)
  ]
  particleSpawnScript [
   &PartitionParticles()
   &InitializeParticle(
    Color Minimum=<rgba 0.5094 0.5094 0.5094 1>
    Color Mode=::NewEnumerator2
    Mesh Scale=&RandomRangeVector(Minimum=<0.25 10 1>, Maximum=<0.3 13 1>)
    Mesh Scale Mode=::NewEnumerator4
    UsePositionOffset=false
    Write Material Random=true
   )
   /SetVariables(Particles.VisibilityTag=&MakeIntFromFloat(Float=$User.Variant_Float)) {isSetParameters=true, module=SetVariables_#6}
   &InitialMeshOrientation(
    Mesh Axis To Orient=<0 0 -1>
    Mesh Orientation Mode=::NewEnumerator2
    Rotation=&RandomRangeVector(Minimum=<0 -0.025 -0.04>, Maximum=<1 0.025 0.04>)
    Use Rotation Vector=true
   )
  ]
  particleUpdateScript [
   &ParticleState(Kill Particles When Lifetime Has Elapsed=false)
   &SkeletalMeshLocation(
    Bone / Socket Index=$Particles.PartitionParticles.Partition
    Inherit Velocity=false
    MaskSpawnGroup=false
    Mesh Sampling Type=::NewEnumerator5
    Sampled Position Offset=&FN_FrontEndVector3Modifier(
     Gameplay=&Lerp_Vector(
      A=&RandomRangeVector(Minimum=<0 -3 0>, Maximum=<3 3 0>)
      B=&RandomRangeVector(Minimum=<0 -3 0>, Maximum=<3 3 0>)
      Alpha=&MakeFloatFromInt(Int=$Particles.PartitionParticles.Partition)
     )
     Front End=&Lerp_Vector(
      A=&RandomRangeVector(Minimum=<0 -3 0>, Maximum=<3 3 0>)
      B=&RandomRangeVector(Minimum=<0 -3 0>, Maximum=<3 3 0>)
      Alpha=&MakeFloatFromInt(Int=$Particles.PartitionParticles.Partition)
     )
    )
    Sampled Position Offset Coordinate Space:ENiagaraMeshTransforms=::NewEnumerator3
    Skeletal Mesh=$User.SkeletalMesh
    Socket Sampling Mode=::NewEnumerator5
   )
   &ScaleMeshSize(Scale Factor=&Lerp_Vector(A=<1 1 1>, B=<2.3 1 1>, Alpha=$System.Timeline.NormalizedTime))
   &UpdateMeshOrientation(
    Facing Interpolation=false
    Orientation Method=::NewEnumerator1
    Override Quat=false
    TransformFacing=true
    TransformUp=false
    Up Direction:Vector3f=&FN_FrontEndVector3Modifier(
     Gameplay=&Lerp_Vector(A=<0 -0.05 1>, B=<0 -0.2 1>, Alpha=$System.Timeline.NormalizedTime)
     Front End=&Lerp_Vector(A=<0 -0.05 1>, B=<0 -0.8 1>, Alpha=$System.Timeline.NormalizedTime)
    )
   )
   &FN_Timer(Increment Multiplier:NiagaraFloat=&Lerp_Float(A=1, B=3, Alpha=$System.Timeline.Time))
   &DynamicMaterialParameters(
    Index 1 Param 2=1
    Index 2 Param 3=1
    Index 2 Param 4=1
    Index 3 Param 1=&Add_Float(A=$Particles.FN_Timer.Value, B=&RandomRangeFloat(Minimum=0, Maximum=10))
    Index 3 Param 4=0.1
    Module.Param0WriteEnabled=true
    Module.Param0WriteEnabled1=true
    Module.Param0WriteEnabled2=true
    Module.Param0WriteEnabled3=true
    Module.Param1WriteEnabled=true
    Module.Param1WriteEnabled1=true
    Module.Param1WriteEnabled2=true
    Module.Param1WriteEnabled3=true
    Module.Param2WriteEnabled=true
    Module.Param2WriteEnabled1=true
    Module.Param2WriteEnabled2=true
    Module.Param2WriteEnabled3=true
    Module.Param3WriteEnabled=true
    Module.Param3WriteEnabled1=true
    Module.Param3WriteEnabled2=true
    Module.Param3WriteEnabled3=true
    Write Parameter Index 1=true
    Write Parameter Index 2=true
    Write Parameter Index 3=true
   )
   &ScaleColor(
    Scale RGB=&VectorFromFloat(Value=&Lerp_Float(A=1, B=3, Alpha=$System.Timeline.NormalizedTime))
    ScaleA=true
    ScaleRGB=true
   )
  ]
  rendererClasses [&NiagaraMeshRendererProperties]
  renderers [{rendererClass=&NiagaraMeshRendererProperties, rendererIndex=0}]
 }
}
summary {
 systemName=NS_Contrail_ZirconSweep
 emitters [
  {emitterName=Dragonflies, rendererClasses [&NiagaraMeshRendererProperties]}
  {emitterName=Trails_Hand, rendererClasses [&NiagaraMeshRendererProperties]}
  {emitterName=Rings, rendererClasses [&NiagaraMeshRendererProperties]}
  {emitterName=Speedlines, simTarget=GPUComputeSim, rendererClasses [&NiagaraSpriteRendererProperties]}
  {emitterName=GlowHands, rendererClasses [&NiagaraSpriteRendererProperties]}
  {emitterName=ParDecal, rendererClasses [&NiagaraSpriteRendererProperties]}
  {emitterName=Trails_Feet, rendererClasses [&NiagaraMeshRendererProperties]}
  {emitterName=GlowFeet, rendererClasses [&NiagaraSpriteRendererProperties]}
  {emitterName=Dragonflies_FE, rendererClasses [&NiagaraMeshRendererProperties]}
 ]
 userVariables [
  User.bIsFrontEnd {type=NiagaraBool}
  User.bIsFrontEndPreview {type=NiagaraBool}
  User.GameTimeFromBP {type=NiagaraFloat, defaultValue=0, description=""}
  User.ParaGlideLeanAlpha {type=NiagaraFloat, defaultValue=0, description=""}
  User.RibbonIDOffset {type=NiagaraInt32}
  User.SkeletalMesh {
   type=NiagaraDataInterfaceSkeletalMesh
   defaultValue {
    struct=NiagaraExt_VariableValue_DataInterface
    value {dataInterface=&self:NiagaraDataInterfaceSkeletalMesh_0}
   }
  }
  User.Variant_Float {type=NiagaraFloat, defaultValue=0, description=""}
 ]
}
